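# Editor utility widget that creates a new season pass plugin from a season number, pass type, plugin version and sunset version.
# On Construct, it fills both version dropdowns from the release version list and the type dropdown from EPassType.
# The create button validates the inputs and checks that /SeasonPassPluginTemplate lists assets, showing warning dialogs on failure.
# It then maps the type string to EPassType, defaulting to BattlePass, and calls CreateSeasonPassPlugin with SeasonPassCopyData.
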EditorUtilityWidgetBlueprint /Game/NeverCook/EditorUtilities/PassTools/EUW_PassPlugin_Copier
graphs [
 EventGraph {nodes [
   "On Clicked (CreatePluginButton)"() {is=title K2Node_ComponentBoundEvent}
   > n1 = ValidateData!(execute, self) -> (then, bIsDataValid:bool)
   > Branch(Condition=@n1)
   > n2 = GetPassTypeFromString!(
    execute
    self
    PassName:String=GetSelectedOption!(self:"Combo Box (String) Object Reference"=Get PluginTypeComboBox() -> (PluginTypeComboBox:Editor Utility Combo Box String Object Reference)) -> (ReturnValue:String)
   ) -> (then, PassType:EPassType Enum=BattlePass)
   n3 = Get SunsetVersionComboBox() -> (SunsetVersionComboBox:Editor Utility Combo Box String Object Reference)
   n4 = "To String (Text)"(InText=Get Text(self=Get SeasonNumberText() -> (SeasonNumberText:Editor Utility Editable Text Box Object Reference)))
   n5 = Get VersionComboBox() -> (VersionComboBox:Editor Utility Combo Box String Object Reference)
   > CreateSeasonPassPlugin!(
    execute
    SeasonPassCopyData:Fort Season Pass Copy Data Object Reference=Get SeasonPassCopyData() -> (SeasonPassCopyData:Fort Season Pass Copy Data Object Reference)
    PassPluginCreateInfo:"Pass Plugin Create Info Structure (by ref)"=Make Pass Plugin Create Info!(
     PassType:EPassType Enum=@n2
     SeasonNumber:int=String To Integer(InString=@n4)
     Version:String=GetSelectedOption(self=@n5)
     SunsetVersion:String=GetSelectedOption(self=@n3)
    ) -> (PassPluginCreateInfo:Pass Plugin Create Info Structure) {is=title K2Node_MakeStruct}
   ) -> (then, ReturnValue:bool)
   Event Construct()
   > n6 = GetAllReleaseVersionsAsString!(execute) -> (then, ReturnValue:Array of Strings)
   > n7 = For Each Loop(Array=@n6)
   n8 = ForEach EPassType!(execute:Exec=@n7.Completed, SkipHidden:bool) -> (LoopBody:Exec, EnumValue:EPassType Enum, then:Exec) {is=title K2Node_ForEachElementInEnum}
   > AddOption!(
    execute
    self:"Combo Box (String) Object Reference"=Get PluginTypeComboBox()
    Option:String=Enum to String!(Enumerator:EPassType Enum=@n8) -> (ReturnValue:String)
   ) -> (then)
   AddOption(execute=@n7.LoopBody, self=Get VersionComboBox(), Option=@n7)
   > AddOption(self=Get SunsetVersionComboBox(), Option=@n7)
   Event PreConstruct()
   Event Tick()
  ]}
 GetPassTypeFromString {nodes [
   n1 = GetPassTypeFromString!() -> (then, PassName:String) {is=title K2Node_FunctionEntry}
   > n2 = Switch on String!(execute:Exec, Selection:String=@n1) -> (Default:Exec, Battle Pass:Exec, Music Pass:Exec, Juno Pass:Exec, OG Pass:Exec) {pinNames [Battle Pass, Music Pass, Juno Pass, OG Pass]}
   Return Node!(execute=[@n2.Battle Pass, @n2.Default], PassType:EPassType Enum=BattlePass)
   Return Node(execute=@n2.Juno Pass, PassType=JunoPass)
   Return Node(execute=@n2.Music Pass, PassType=MusicPass)
   Return Node(execute=@n2.OG Pass, PassType=OGPass)
  ]}
 ValidateData {nodes [
   ValidateData() {is=title K2Node_FunctionEntry}
   n1 = GetSelectedOption(self=Get SunsetVersionComboBox())
   n2 = GetSelectedOption(self=Get VersionComboBox())
   n3 = GetSelectedOption(self=Get PluginTypeComboBox())
   n4 = Get SeasonNumberText()
   n5 = IsNumeric!(SourceString:String="To String (Text)"(InText=Get Text(self=@n4))) -> (ReturnValue:bool)
   > n6 = Branch(Condition=OR Boolean!(A=NOT Boolean(A=@n5), B=IsEmpty(InString=@n3), C:bool=IsEmpty(InString=@n2), D:bool=IsEmpty(InString=@n1)) -> (ReturnValue))
   > Show Message Dialog(
    Title=MakeLiteralText(Value="LOCTEXT(\"There are issues with the data.\")")
    Message="To Text (String)"(InString=MakeLiteralString(Value="There could be an issue with:\r\n\r\n-Season Number is wrong.\r\n-Pass Type is not a valid type.\r\n-Version is not valid.\r\n-Sunset Version is not valid."))
    MessageType=Ok
    DefaultValue=No
    MessageCategory=Warning
   )
   > Return Node!(execute, bIsDataValid:bool=false)
   n7 = ListAssets!(execute=@n6.else, DirectoryPath:String=/SeasonPassPluginTemplate, bRecursive:bool=true, bIncludeFolder:bool=true) -> (then, ReturnValue:Array of Strings)
   > n8 = Branch(Condition=NOT Boolean(A=Is Empty!(TargetArray=@n7) -> (ReturnValue:bool)))
   > Return Node(bIsDataValid=true)
   Show Message Dialog(
    execute=@n8.else
    Title=MakeLiteralText(Value="LOCTEXT(\"Season Pass Plugin Template is not monted!\")")
    Message="To Text (String)"(InString=MakeLiteralString(Value="Season Pass Plugin Template must be mounted and it's not.\r\n\r\nIn order to mount it open Tools >Plugin Template Tool and search for Season Pass Template plugin, click on mount.  "))
    MessageType=Ok
    DefaultValue=No
    MessageCategory=Warning
   )
  ]}
]
variables [SeasonPassCopyData {type=FortSeasonPassCopyData, default=&/SeasonPassCopyData}]
widgets {
 info {rootWidgetClass=&VerticalBox, widgetCount=14}
 widgets [
  VerticalBox_165 {widgetClassPath=&VerticalBox, widget=VerticalBox_165}
  HorizontalBox_6 {widgetClassPath=&HorizontalBox, parent=VerticalBox_165, widget=HorizontalBox_6, slot {class=VerticalBoxSlot}}
  Seasonnumber {
   widgetClassPath=&TextBlock
   parent=HorizontalBox_6
   widget=Seasonnumber
   properties {
    text="New Season Number: "
    font {
     bForceMonospaced=false
     bMaterialIsStencil=false
     fontMaterial=null
     fontObject=&Roboto
     letterSpacing=0
     monospacedWidth=1
     size=12
     skewAmount=0
     typefaceFontName=Bold
     outlineSettings {
      bApplyOutlineToDropShadows=false
      bMiteredCorners=false
      bSeparateFillAlpha=false
      outlineMaterial=null
      outlineSize=0
      outlineColor=<rgba 0 0 0 1>
     }
    }
   }
   slot {class=HorizontalBoxSlot}
  }
  SeasonNumberText {
   widgetClassPath=&EditorUtilityEditableTextBox
   bIsVariable=true
   parent=HorizontalBox_6
   widget=SeasonNumberText
   properties {text=00}
   slot {class=HorizontalBoxSlot}
  }
  HorizontalBox_4 {widgetClassPath=&HorizontalBox, parent=VerticalBox_165, widget=HorizontalBox_4, slot {class=VerticalBoxSlot}}
  Plugintype {
   widgetClassPath=&TextBlock
   parent=HorizontalBox_4
   widget=Plugintype
   properties {
    text="Plugin Type: "
    font {
     bForceMonospaced=false
     bMaterialIsStencil=false
     fontMaterial=null
     fontObject=&Roboto
     letterSpacing=0
     monospacedWidth=1
     size=12
     skewAmount=0
     typefaceFontName=Bold
     outlineSettings {
      bApplyOutlineToDropShadows=false
      bMiteredCorners=false
      bSeparateFillAlpha=false
      outlineMaterial=null
      outlineSize=0
      outlineColor=<rgba 0 0 0 1>
     }
    }
   }
   slot {class=HorizontalBoxSlot}
  }
  PluginTypeComboBox {
   widgetClassPath=/Script/Blutility.EditorUtilityComboBoxString
   bIsVariable=true
   parent=HorizontalBox_4
   widget=PluginTypeComboBox
   properties {
    font {
     bForceMonospaced=false
     bMaterialIsStencil=false
     fontMaterial=null
     fontObject=&Roboto
     letterSpacing=0
     monospacedWidth=1
     size=12
     skewAmount=0
     typefaceFontName=Bold
     outlineSettings {
      bApplyOutlineToDropShadows=false
      bMiteredCorners=false
      bSeparateFillAlpha=false
      outlineMaterial=null
      outlineSize=0
      outlineColor=<rgba 0 0 0 1>
     }
    }
    foregroundColor {colorUseRule=UseColor_Specified, specifiedColor=<rgba 0.5271 0.5271 0.5271 1>}
   }
   slot {class=HorizontalBoxSlot}
  }
  HorizontalBox_3 {widgetClassPath=&HorizontalBox, parent=VerticalBox_165, widget=HorizontalBox_3, slot {class=VerticalBoxSlot}}
  Version {
   widgetClassPath=&TextBlock
   parent=HorizontalBox_3
   widget=Version
   properties {
    text="Plugin Version: "
    font {
     bForceMonospaced=false
     bMaterialIsStencil=false
     fontMaterial=null
     fontObject=&Roboto
     letterSpacing=0
     monospacedWidth=1
     size=12
     skewAmount=0
     typefaceFontName=Bold
     outlineSettings {
      bApplyOutlineToDropShadows=false
      bMiteredCorners=false
      bSeparateFillAlpha=false
      outlineMaterial=null
      outlineSize=0
      outlineColor=<rgba 0 0 0 1>
     }
    }
   }
   slot {class=HorizontalBoxSlot}
  }
  VersionComboBox {
   widgetClassPath=&EditorUtilityComboBoxString
   bIsVariable=true
   parent=HorizontalBox_3
   widget=VersionComboBox
   properties {
    font {
     bForceMonospaced=false
     bMaterialIsStencil=false
     fontMaterial=null
     fontObject=&Roboto
     letterSpacing=0
     monospacedWidth=1
     size=12
     skewAmount=0
     typefaceFontName=Bold
     outlineSettings {
      bApplyOutlineToDropShadows=false
      bMiteredCorners=false
      bSeparateFillAlpha=false
      outlineMaterial=null
      outlineSize=0
      outlineColor=<rgba 0 0 0 1>
     }
    }
    foregroundColor {colorUseRule=UseColor_Specified, specifiedColor=<rgba 0.5271 0.5271 0.5271 1>}
   }
   slot {class=HorizontalBoxSlot}
  }
  HorizontalBox_5 {widgetClassPath=&HorizontalBox, parent=VerticalBox_165, widget=HorizontalBox_5, slot {class=VerticalBoxSlot}}
  SunsetVersion {
   widgetClassPath=&TextBlock
   parent=HorizontalBox_5
   widget=SunsetVersion
   properties {
    text="Sunset Version: "
    font {
     bForceMonospaced=false
     bMaterialIsStencil=false
     fontMaterial=null
     fontObject=&Roboto
     letterSpacing=0
     monospacedWidth=1
     size=12
     skewAmount=0
     typefaceFontName=Bold
     outlineSettings {
      bApplyOutlineToDropShadows=false
      bMiteredCorners=false
      bSeparateFillAlpha=false
      outlineMaterial=null
      outlineSize=0
      outlineColor=<rgba 0 0 0 1>
     }
    }
   }
   slot {class=HorizontalBoxSlot}
  }
  SunsetVersionComboBox {
   widgetClassPath=&EditorUtilityComboBoxString
   bIsVariable=true
   parent=HorizontalBox_5
   widget=SunsetVersionComboBox
   properties {
    font {
     bForceMonospaced=false
     bMaterialIsStencil=false
     fontMaterial=null
     fontObject=&Roboto
     letterSpacing=0
     monospacedWidth=1
     size=12
     skewAmount=0
     typefaceFontName=Bold
     outlineSettings {
      bApplyOutlineToDropShadows=false
      bMiteredCorners=false
      bSeparateFillAlpha=false
      outlineMaterial=null
      outlineSize=0
      outlineColor=<rgba 0 0 0 1>
     }
    }
    foregroundColor {colorUseRule=UseColor_Specified, specifiedColor=<rgba 0.5271 0.5271 0.5271 1>}
   }
   slot {class=HorizontalBoxSlot}
  }
  CreatePluginButton {widgetClassPath=&EditorUtilityButton, parent=VerticalBox_165, widget=CreatePluginButton, slot {class=VerticalBoxSlot}}
 ]
}
