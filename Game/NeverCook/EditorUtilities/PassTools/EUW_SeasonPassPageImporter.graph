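# Editor Utility Widget that serves as a front end for importing battle pass season pages from a CSV spreadsheet and for generating a template.
# It takes a Content Browser pass root path and a spreadsheet file path from two text boxes.
# Each button runs embedded Python that reloads SeasonPass.SeasonPassPageImporter, then calls run_import or generate_template with both paths.
# The returned result text replaces the contents of a read-only log box.
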EditorUtilityWidgetBlueprint /Game/NeverCook/EditorUtilities/PassTools/EUW_SeasonPassPageImporter
graphs [EventGraph {nodes [
   "On Clicked (Btn_Import)"() {is=title K2Node_ComponentBoundEvent}
   n1 = Get Text(self=Get Input_SpreadsheetPath() -> (Input_SpreadsheetPath:Editor Utility Editable Text Box Object Reference))
   n2 = Get Text(self=Get Input_PassRootPath() -> (Input_PassRootPath:Editor Utility Editable Text Box Object Reference))
   > n3 = Execute Python Script!(
    execute
    PythonScript="import importlib\r\nimport SeasonPass.SeasonPassPageImporter as M\r\nimportlib.reload(M)\r\nresult = M.run_import(spreadsheet_path, pass_root)"
    spreadsheet_path:Text=@n1
    pass_root:Text=@n2
   ) -> (then, ReturnValue, result:Text) {inputs [spreadsheet_path, pass_root], outputs [result]}
   > Set Text(
    Text=@n3.result
    self:"Text Box (Multi- Line) Object Reference"=Get Output_Log() -> (Output_Log:Editor Utility Multi Line Editable Text Box Object Reference)
   )
   "On Clicked (Btn_GenerateTemplate)"() {is=title K2Node_ComponentBoundEvent}
   > n4 = Execute Python Script(
    PythonScript="import importlib\r\nimport SeasonPass.SeasonPassPageImporter as M\r\nimportlib.reload(M)\r\nresult = M.generate_template(spreadsheet_path, pass_root)"
    spreadsheet_path=@n1
    pass_root=@n2
   ) {inputs [spreadsheet_path, pass_root], outputs [result]}
   > Set Text(Text=@n4.result, self=Get Output_Log())
   Event PreConstruct()
   Event Construct()
   Event Tick()
  ]}]
variables [LastGeneratedPath {type=string, default=""}]
widgets {
 info {rootWidgetClass=&VerticalBox, widgetCount=12}
 widgets [
  VerticalBox_0 {widgetClassPath=&VerticalBox, widget=VerticalBox_0}
  TextBlock_145 {
   widgetClassPath=&TextBlock
   parent=VerticalBox_0
   widget=TextBlock_145
   properties {text="Pass Root Path (Content Browser)"}
   slot {class=VerticalBoxSlot}
  }
  Input_PassRootPath {
   widgetClassPath=&EditorUtilityEditableTextBox
   bIsVariable=true
   parent=VerticalBox_0
   widget=Input_PassRootPath
   properties {hintText=/BattlePassS40_SeasonPass/SeasonPass/Categories}
   slot {class=VerticalBoxSlot}
  }
  TextBlock_229 {
   widgetClassPath=&TextBlock
   parent=VerticalBox_0
   widget=TextBlock_229
   properties {text="Spreadsheet File (.csv)"}
   slot {class=VerticalBoxSlot}
  }
  Input_SpreadsheetPath {
   widgetClassPath=&EditorUtilityEditableTextBox
   bIsVariable=true
   parent=VerticalBox_0
   widget=Input_SpreadsheetPath
   properties {hintText="C:\\MySheets\\BattlePass_S40.csv"}
   slot {class=VerticalBoxSlot}
  }
  HorizontalBox_125 {widgetClassPath=&HorizontalBox, parent=VerticalBox_0, widget=HorizontalBox_125, slot {class=VerticalBoxSlot}}
  Btn_GenerateTemplate {
   widgetClassPath=&EditorUtilityButton
   parent=HorizontalBox_125
   widget=Btn_GenerateTemplate
   slot {class=HorizontalBoxSlot, properties {padding=<ltrb 5 5 5 5>}}
  }
  TextBlock_351 {
   widgetClassPath=&TextBlock
   parent=Btn_GenerateTemplate
   widget=TextBlock_351
   properties {text=Generate Template}
   slot {class=ButtonSlot}
  }
  Btn_Import {
   widgetClassPath=&EditorUtilityButton
   parent=HorizontalBox_125
   widget=Btn_Import
   slot {class=HorizontalBoxSlot, properties {padding=<ltrb 5 5 5 5>}}
  }
  TextBlock_542 {widgetClassPath=&TextBlock, parent=Btn_Import, widget=TextBlock_542, properties {text=Run Import}, slot {class=ButtonSlot}}
  TextBlock_1089 {
   widgetClassPath=&TextBlock
   parent=VerticalBox_0
   widget=TextBlock_1089
   properties {text=Log}
   slot {class=VerticalBoxSlot, properties {padding=<ltrb 0 30 0 0>}}
  }
  Output_Log {
   widgetClassPath=/Script/Blutility.EditorUtilityMultiLineEditableTextBox
   bIsVariable=true
   parent=VerticalBox_0
   widget=Output_Log
   properties {bIsReadOnly=true}
   slot {class=VerticalBoxSlot, properties {size {sizeRule=Fill, value=1}}}
  }
 ]
}
